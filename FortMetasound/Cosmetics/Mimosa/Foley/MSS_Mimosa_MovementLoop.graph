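# Mono movement foley loop source for the Mimosa cosmetic companion, layering a random intro with a random looping wave.
# The loop starts on play or when the intro finishes, per a template flag, after an optional delay, with an attack-envelope fade-in.
# A second flag routes interpolated movement speed, range-mapped, to loop pitch and volume, otherwise both use unity.
# Both layers share a pitch multiplier and mix with separate gains.
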MetaSoundSource /FortMetasound/Cosmetics/Mimosa/Foley/MSS_Mimosa_MovementLoop
inputs [
 Foley.Sounds.Intro {type=WaveAsset:Array}
 Foley.Sounds.Loop {type=WaveAsset:Array}
 Intro.VolumeMultiplier {type=Float}
 loop.DelayTime {type=Time}
 Loop.FadeInTime {type=Time}
 Loop.VolumeMultiplier {type=Float}
 Movement.Speed {type=Float}
 Movement.SpeedAzimuth {type=Float}
 Movement.SpeedElevation {type=Float}
 Speed.InterpTime {type=Time}
 Speed.MaxPitch {type=Float}
 Speed.MaxTracked {type=Float}
 Speed.MaxVolume {type=Float}
 Speed.MinPitch {type=Float}
 speed.MinTracked {type=Float}
 Speed.MinVolume {type=Float}
 Template.bControlSoundWithSpeed {type=Bool}
 Template.bDelayLoopByIntro {type=Bool}
 Template.PitchMultiplier {type=Float}
 TestSpeed {type=Float}
 UE.Source.OnPlay {type=Trigger}
]
interfaces [Fort.Movement:1.0, UE.OutputFormat.Mono:1.0, UE.Source:1.1]
outputs ["UE.OutputFormat.Mono.Audio:0" {type=Audio}]
pages [Default {
  input_defaults {
   Intro.VolumeMultiplier=1
   Loop.FadeInTime=0
   Loop.VolumeMultiplier=1
   Movement.Speed=0
   Movement.SpeedAzimuth=0
   Movement.SpeedElevation=0
   Speed.InterpTime=0.1
   Speed.MaxPitch=1.2
   Speed.MaxTracked=1000
   Speed.MaxVolume=1
   Speed.MinPitch=0.8
   Speed.MinVolume=0.6
   Template.PitchMultiplier=1
   Template.bControlSoundWithSpeed="<MetaSoundBuilderResult.SUCCEEDED: 0>"
   Template.bDelayLoopByIntro="<MetaSoundBuilderResult.SUCCEEDED: 0>"
   TestSpeed=300
   UE.Source.OnPlay=null
   loop.DelayTime=0
   speed.MinTracked=300
   Foley.Sounds.Intro []
   Foley.Sounds.Loop []
  }
  nodes [
   n1 = InitVariable.Audio()
   n2 = InitVariable.Trigger()
   ArrayRandomGet_0 = Array.Random Get:WaveAsset:Array(In Array=^Foley.Sounds.Intro, Next=^UE.Source.OnPlay)
   FrequencyMultipliertoSemitone_1 = UE.Frequency Multiplier to Semitone:Float(Frequency Multiplier=^Template.PitchMultiplier)
   WavePlayer_1 = UE.Wave Player:Mono(
    Play=@ArrayRandomGet_0.On Next
    Wave Asset=@ArrayRandomGet_0.Value
    Pitch Shift=UE.Reroute(Value=UE.Reroute(Value=@FrequencyMultipliertoSemitone_1.Semitones).Value).Value
   ) {also {version [1,1]}}
   n3 = VariableMutator.Trigger(Variable=InitVariable.Trigger().Variable, Value=@WavePlayer_1.On Finished)
   TriggerAnyTriggerAny2_0 = "TriggerAny.Trigger Any (2)"(
    In 0=TriggerCompare.Clamp:Bool(A=^Template.bDelayLoopByIntro, Compare=^UE.Source.OnPlay) {also {defaults {B="<MetaSoundBuilderResult.SUCCEEDED: 0>"}}}.False
    In 1=TriggerCompare.Clamp:Bool(A=^Template.bDelayLoopByIntro, Compare=VariableAccessor.Trigger(Variable=@n3.Variable).Value) {also {defaults {B="<MetaSoundBuilderResult.SUCCEEDED: 0>"}}}.True
   )
   TriggerDelay_0 = UE.Trigger Delay(
    Delay Time=^loop.DelayTime
    In=VariableAccessor.Trigger(Variable=VariableMutator.Trigger(Variable=@n2.Variable, Value=@TriggerAnyTriggerAny2_0.Out).Variable).Value
   )
   ArrayRandomGet_1 = Array.Random Get:WaveAsset:Array(In Array=^Foley.Sounds.Loop, Next=@TriggerDelay_0.Out) {also {defaults {No Repeats=-1}}}
   n4 = TriggerCompare.Clamp:Bool(A=^Template.bControlSoundWithSpeed, Compare=@ArrayRandomGet_1.On Next) {also {defaults {B="<MetaSoundBuilderResult.SUCCEEDED: 0>"}}}
   n5 = InitVariable.Float()
   n6 = MapRange.MapRange:Float(
    Out Range B=^Speed.MaxPitch
    In Range B=^Speed.MaxTracked
    Out Range A=^Speed.MinPitch
    In Range A=^speed.MinTracked
    In=UE.InterpTo:Audio(Target=^Movement.Speed, Interp Time=^Speed.InterpTime).Value
   )
   TriggerRouteTriggerRouteFloat2_0 = "TriggerRoute.Trigger Route (Float, 2):Float"(
    Set 1=@n4.False
    Set 0=@n4.True
    Value 0=VariableAccessor.Float(Variable=VariableMutator.Float(Variable=@n5.Variable, Value=@n6.Out Value).Variable).Value
   ) {also {defaults {Value 1=1}}}
   FrequencyMultipliertoSemitone_0 = UE.Frequency Multiplier to Semitone:Float(Frequency Multiplier=UE.Multiply:Float(AdditionalOperands=^Template.PitchMultiplier, PrimaryOperand=@TriggerRouteTriggerRouteFloat2_0.Value).Out)
   n7 = UE.Reroute(Value=UE.Reroute(Value=@ArrayRandomGet_1.Value).Value)
   GetWaveDuration_0 = "UE.Get Wave Duration: "(Wave=@n7.Value)
   WavePlayer_0 = UE.Wave Player:Mono(
    Pitch Shift=@FrequencyMultipliertoSemitone_0.Semitones
    Play=UE.RandomTime(Max=@GetWaveDuration_0.Duration, Next=@TriggerRouteTriggerRouteFloat2_0.On Set).On Next
    Wave Asset=UE.Reroute(Value=@n7.Value).Value
   ) {also {defaults {Loop="<MetaSoundBuilderResult.SUCCEEDED: 0>"}, version [1,1]}}
   ADSREnvelopeADSREnvelope_0 = ADSR Envelope.ADSR Envelope:Float(Attack Time=^Loop.FadeInTime, Trigger Attack=@WavePlayer_0.On Play) {also {defaults {Decay Time=0, Hard Reset="<MetaSoundBuilderResult.SUCCEEDED: 0>", Sustain Level=1}}}
   n8 = UE.Reroute(Value=@ADSREnvelopeADSREnvelope_0.Out Envelope)
   n9 = TriggerCompare.Clamp:Bool(A=^Template.bControlSoundWithSpeed, Compare=@ADSREnvelopeADSREnvelope_0.On Attack Triggered) {also {defaults {B="<MetaSoundBuilderResult.SUCCEEDED: 0>"}}}
   n10 = InitVariable.Float()
   n11 = MapRange.MapRange:Float(
    In Range B=^Speed.MaxTracked
    Out Range B=^Speed.MaxVolume
    Out Range A=^Speed.MinVolume
    In Range A=^speed.MinTracked
    In=UE.InterpTo:Audio(Target=^Movement.Speed, Interp Time=^Speed.InterpTime).Value
   )
   TriggerRouteTriggerRouteFloat2_1 = "TriggerRoute.Trigger Route (Float, 2):Float"(
    Set 1=@n9.False
    Set 0=@n9.True
    Value 0=VariableAccessor.Float(Variable=VariableMutator.Float(Variable=@n10.Variable, Value=@n11.Out Value).Variable).Value
   ) {also {defaults {Value 1=1}}}
   n12 = UE.Multiply:Audio by Float(
    AdditionalOperands=UE.Multiply:Float(AdditionalOperands=UE.Reroute(Value=@n8.Value).Value, PrimaryOperand=@TriggerRouteTriggerRouteFloat2_1.Value).Out
    PrimaryOperand=UE.Reroute(Value=UE.Reroute(Value=@WavePlayer_0.Out Mono).Value).Value
   )
   n13 = InitVariable.Audio()
   AudioMixerAudioMixerMono2_0 = "AudioMixer.Audio Mixer (Mono, 2)"(
    Gain 0=^Intro.VolumeMultiplier
    Gain 1=^Loop.VolumeMultiplier
    In 1=VariableAccessor.Audio(Variable=VariableMutator.Audio(Variable=@n1.Variable, Value=@n12.Out).Variable).Value
    In 0=VariableAccessor.Audio(Variable=VariableMutator.Audio(Variable=@n13.Variable, Value=@WavePlayer_1.Out Mono).Variable).Value
   )
  ]
  links [{dst="@UE.OutputFormat.Mono.Audio:0", src=AudioMixerAudioMixerMono2_0.Out}]
 }]
properties {
 assetImportData=&self:AssetImportData
 attenuationSettings=/Game/Sounds/Attenuation/AS_Companion_Foley
 audioPropertiesSheet=/CosmeticCompanions/Assets/Audio_Shared/PropertySheets/APS_Cosmetic_Mimosa_FoleyLoop
 bLooping=true
 bOverrideConcurrency=true
 duration=10000
 maxDistance=1200
 sampleRate=48000
 soundClassObject=&Fort_Creatures_Pets
 concurrencyOverrides {bLimitToOwner=true, maxCount=1}
}
