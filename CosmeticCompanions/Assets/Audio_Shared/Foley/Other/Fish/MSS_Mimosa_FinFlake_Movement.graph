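# Mono MetaSound Source preset that plays looping fin-flake movement foley for a fish cosmetic companion.
# It inherits its graph from MSS_Mimosa_MovementLoop, supplying eight loop wave variations and an empty intro array.
# Parent inputs expose Fort.Movement speed with tracked-range, smoothing, pitch and volume limits, plus loop delay and fade-in.
# It uses companion foley attenuation, the pets sound class and a concurrency override of one instance per owner.
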MetaSoundSource /CosmeticCompanions/Assets/Audio_Shared/Foley/Other/Fish/MSS_Mimosa_FinFlake_Movement
preset_parent=/FortMetasound/Cosmetics/Mimosa/Foley/MSS_Mimosa_MovementLoop
inputs [
 Foley.Sounds.Intro {type=WaveAsset:Array}
 Foley.Sounds.Loop {type=WaveAsset:Array}
 Intro.VolumeMultiplier {type=Float}
 loop.DelayTime {type=Time}
 Loop.FadeInTime {type=Time}
 Loop.VolumeMultiplier {type=Float}
 Movement.Speed {type=Float}
 Movement.SpeedAzimuth {type=Float}
 Movement.SpeedElevation {type=Float}
 Speed.InterpTime {type=Time}
 Speed.MaxPitch {type=Float}
 Speed.MaxTracked {type=Float}
 Speed.MaxVolume {type=Float}
 Speed.MinPitch {type=Float}
 speed.MinTracked {type=Float}
 Speed.MinVolume {type=Float}
 Template.bControlSoundWithSpeed {type=Bool}
 Template.bDelayLoopByIntro {type=Bool}
 Template.PitchMultiplier {type=Float}
 TestSpeed {type=Float}
 UE.Source.OnPlay {type=Trigger}
]
interfaces [Fort.Movement:1.0, UE.OutputFormat.Mono:1.0, UE.Source:1.1]
outputs ["UE.OutputFormat.Mono.Audio:0" {type=Audio}]
pages [Default {input_defaults {
   Intro.VolumeMultiplier=1
   Loop.FadeInTime=0.2
   Loop.VolumeMultiplier=1
   Movement.Speed=0
   Movement.SpeedAzimuth=0
   Movement.SpeedElevation=0
   Speed.InterpTime=0.1
   Speed.MaxPitch=1.2
   Speed.MaxTracked=1000
   Speed.MaxVolume=1
   Speed.MinPitch=0.8
   Speed.MinVolume=0.6
   Template.PitchMultiplier=1
   Template.bControlSoundWithSpeed="<MetaSoundBuilderResult.SUCCEEDED: 0>"
   Template.bDelayLoopByIntro="<MetaSoundBuilderResult.SUCCEEDED: 0>"
   TestSpeed=300
   UE.Source.OnPlay=null
   loop.DelayTime=0
   speed.MinTracked=300
   Foley.Sounds.Intro []
   Foley.Sounds.Loop [
    /CosmeticCompanions/Assets/Audio_Shared/Foley/Other/Fish/Wavs/Mimosa_FinFlake_MovementLoop_010
    /CosmeticCompanions/Assets/Audio_Shared/Foley/Other/Fish/Wavs/Mimosa_FinFlake_MovementLoop_017
    /CosmeticCompanions/Assets/Audio_Shared/Foley/Other/Fish/Wavs/Mimosa_FinFlake_MovementLoop_016
    /CosmeticCompanions/Assets/Audio_Shared/Foley/Other/Fish/Wavs/Mimosa_FinFlake_MovementLoop_015
    /CosmeticCompanions/Assets/Audio_Shared/Foley/Other/Fish/Wavs/Mimosa_FinFlake_MovementLoop_014
    /CosmeticCompanions/Assets/Audio_Shared/Foley/Other/Fish/Wavs/Mimosa_FinFlake_MovementLoop_013
    /CosmeticCompanions/Assets/Audio_Shared/Foley/Other/Fish/Wavs/Mimosa_FinFlake_MovementLoop_012
    /CosmeticCompanions/Assets/Audio_Shared/Foley/Other/Fish/Wavs/Mimosa_FinFlake_MovementLoop_011
   ]
  }}]
properties {
 assetImportData=&self:AssetImportData
 attenuationSettings=/Game/Sounds/Attenuation/AS_Companion_Foley
 audioPropertiesSheet=/CosmeticCompanions/Assets/Audio_Shared/PropertySheets/APS_Cosmetic_Mimosa_FoleyLoop
 bLooping=true
 bOverrideConcurrency=true
 duration=10000
 maxDistance=1200
 sampleRate=48000
 soundClassObject=&Fort_Creatures_Pets
 volume=2.7
 concurrencyOverrides {bLimitToOwner=true, maxCount=1}
}
