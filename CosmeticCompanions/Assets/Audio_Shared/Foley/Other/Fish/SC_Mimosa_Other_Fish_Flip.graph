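# Sound cue for fish flip foley on a cosmetic companion, classed as a creature pet sound.
# A quality level node sends its first two levels to a random pick of five flip waves and its last two to a pick of the first four.
# All picks are equally weighted.
# Playback uses companion foley attenuation, zero priority and the quadruped footstep concurrency set.
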SoundCue /CosmeticCompanions/Assets/Audio_Shared/Foley/Other/Fish/SC_Mimosa_Other_Fish_Flip
root=QualityLevel_0
nodes [
 {id=QualityLevel_0, children [Random_0,Random_0,Random_1,Random_1]}
 {id=Random_0, children [WavePlayer_2,WavePlayer_3,WavePlayer_4,WavePlayer_5,WavePlayer_6], properties {weights [1,1,1,1,1]}}
 {id=Random_1, children [WavePlayer_7, WavePlayer_8, WavePlayer_9, WavePlayer_18], properties {weights [1,1,1,1]}}
 {id=WavePlayer_18, properties {soundWaveAssetPtr=/CosmeticCompanions/Assets/Audio_Shared/Foley/Other/Fish/Wavs/Mimosa_Fish_Flip_04}}
 {id=WavePlayer_2, properties {soundWaveAssetPtr=/CosmeticCompanions/Assets/Audio_Shared/Foley/Other/Fish/Wavs/Mimosa_Fish_Flip_01}}
 {id=WavePlayer_3, properties {soundWaveAssetPtr=/CosmeticCompanions/Assets/Audio_Shared/Foley/Other/Fish/Wavs/Mimosa_Fish_Flip_02}}
 {id=WavePlayer_4, properties {soundWaveAssetPtr=/CosmeticCompanions/Assets/Audio_Shared/Foley/Other/Fish/Wavs/Mimosa_Fish_Flip_03}}
 {id=WavePlayer_5, properties {soundWaveAssetPtr=&Mimosa_Fish_Flip_04}}
 {id=WavePlayer_6, properties {soundWaveAssetPtr=/CosmeticCompanions/Assets/Audio_Shared/Foley/Other/Fish/Wavs/Mimosa_Fish_Flip_05}}
 {id=WavePlayer_7, properties {soundWaveAssetPtr=&Mimosa_Fish_Flip_01}}
 {id=WavePlayer_8, properties {soundWaveAssetPtr=&Mimosa_Fish_Flip_02}}
 {id=WavePlayer_9, properties {soundWaveAssetPtr=&Mimosa_Fish_Flip_03}}
]
properties {
 attenuationSettings=/Game/Sounds/Attenuation/AS_Companion_Foley
 maxDistance=1200
 priority=0
 soundClassObject=&Fort_Creatures_Pets
 volumeMultiplier=0.8
 concurrencySet [/CosmeticCompanions/Assets/Audio_Shared/Foley/Quadraped/CONC_Mimosa_Quadraped_Footstep]
}
