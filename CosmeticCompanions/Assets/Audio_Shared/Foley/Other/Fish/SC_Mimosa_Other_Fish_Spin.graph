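# Sound cue for the spinning foley of the Mimosa fish cosmetic companion.
# A quality-level node sends the first two tiers to a random pick of five spin waves and the last two to a random pick of the first four, with equal weights.
# It uses companion foley attenuation and the pets sound class, and borrows the quadruped footstep concurrency set.
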SoundCue /CosmeticCompanions/Assets/Audio_Shared/Foley/Other/Fish/SC_Mimosa_Other_Fish_Spin
root=QualityLevel_0
nodes [
 {id=QualityLevel_0, children [Random_0,Random_0,Random_3,Random_3]}
 {id=Random_0, children [WavePlayer_0, WavePlayer_11, WavePlayer_12, WavePlayer_13, WavePlayer_14], properties {weights [1,1,1,1,1]}}
 {id=Random_3, children [WavePlayer_15, WavePlayer_16, WavePlayer_17, WavePlayer_18], properties {weights [1,1,1,1]}}
 {id=WavePlayer_0, properties {soundWaveAssetPtr=/CosmeticCompanions/Assets/Audio_Shared/Foley/Other/Fish/Wavs/Mimosa_Fish_Spin_01}}
 {id=WavePlayer_11, properties {soundWaveAssetPtr=/CosmeticCompanions/Assets/Audio_Shared/Foley/Other/Fish/Wavs/Mimosa_Fish_Spin_02}}
 {id=WavePlayer_12, properties {soundWaveAssetPtr=/CosmeticCompanions/Assets/Audio_Shared/Foley/Other/Fish/Wavs/Mimosa_Fish_Spin_03}}
 {id=WavePlayer_13, properties {soundWaveAssetPtr=/CosmeticCompanions/Assets/Audio_Shared/Foley/Other/Fish/Wavs/Mimosa_Fish_Spin_04}}
 {id=WavePlayer_14, properties {soundWaveAssetPtr=/CosmeticCompanions/Assets/Audio_Shared/Foley/Other/Fish/Wavs/Mimosa_Fish_Spin_05}}
 {id=WavePlayer_15, properties {soundWaveAssetPtr=&Mimosa_Fish_Spin_01}}
 {id=WavePlayer_16, properties {soundWaveAssetPtr=&Mimosa_Fish_Spin_02}}
 {id=WavePlayer_17, properties {soundWaveAssetPtr=&Mimosa_Fish_Spin_03}}
 {id=WavePlayer_18, properties {soundWaveAssetPtr=&Mimosa_Fish_Spin_04}}
]
properties {
 attenuationSettings=/Game/Sounds/Attenuation/AS_Companion_Foley
 maxDistance=1200
 priority=0
 soundClassObject=&Fort_Creatures_Pets
 volumeMultiplier=1.7
 concurrencySet [/CosmeticCompanions/Assets/Audio_Shared/Foley/Quadraped/CONC_Mimosa_Quadraped_Footstep]
}
