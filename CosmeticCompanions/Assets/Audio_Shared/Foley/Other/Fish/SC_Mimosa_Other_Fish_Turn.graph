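# Sound cue for the turn foley of a fish cosmetic companion, routed to the pets sound class with companion foley attenuation.
# A quality switch sends its first two levels to an equal-weight random pick of six turn waves and its last two to a smaller pool of four.
# A root modulator varies pitch and volume on each play.
# The cue runs at zero priority and shares the quadruped footstep concurrency set.
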SoundCue /CosmeticCompanions/Assets/Audio_Shared/Foley/Other/Fish/SC_Mimosa_Other_Fish_Turn
root=Modulator_0
nodes [
 {id=Modulator_0, children [QualityLevel_0], properties {pitchMax=1.1, pitchMin=0.8, volumeMax=1.1, volumeMin=0.7}}
 {id=QualityLevel_0, children [Random_3,Random_3,Random_5,Random_5]}
 {id=Random_3, children [WavePlayer_0,WavePlayer_1,WavePlayer_2,WavePlayer_3,WavePlayer_4,WavePlayer_5], properties {weights [1,1,1,1,1,1]}}
 {id=Random_5, children [WavePlayer_6, WavePlayer_7, WavePlayer_22, WavePlayer_23], properties {weights [1,1,1,1]}}
 {id=WavePlayer_0, properties {soundWaveAssetPtr=/CosmeticCompanions/Assets/Audio_Shared/Foley/Other/Fish/Wavs/Mimosa_Fish_Turn_01}}
 {id=WavePlayer_1, properties {soundWaveAssetPtr=/CosmeticCompanions/Assets/Audio_Shared/Foley/Other/Fish/Wavs/Mimosa_Fish_Turn_02}}
 {id=WavePlayer_2, properties {soundWaveAssetPtr=/CosmeticCompanions/Assets/Audio_Shared/Foley/Other/Fish/Wavs/Mimosa_Fish_Turn_03}}
 {id=WavePlayer_22, properties {soundWaveAssetPtr=&Mimosa_Fish_Turn_03}}
 {id=WavePlayer_23, properties {soundWaveAssetPtr=/CosmeticCompanions/Assets/Audio_Shared/Foley/Other/Fish/Wavs/Mimosa_Fish_Turn_04}}
 {id=WavePlayer_3, properties {soundWaveAssetPtr=&Mimosa_Fish_Turn_04}}
 {id=WavePlayer_4, properties {soundWaveAssetPtr=/CosmeticCompanions/Assets/Audio_Shared/Foley/Other/Fish/Wavs/Mimosa_Fish_Turn_05}}
 {id=WavePlayer_5, properties {soundWaveAssetPtr=/CosmeticCompanions/Assets/Audio_Shared/Foley/Other/Fish/Wavs/Mimosa_Fish_Turn_06}}
 {id=WavePlayer_6, properties {soundWaveAssetPtr=&Mimosa_Fish_Turn_01}}
 {id=WavePlayer_7, properties {soundWaveAssetPtr=&Mimosa_Fish_Turn_02}}
]
properties {
 attenuationSettings=/Game/Sounds/Attenuation/AS_Companion_Foley
 maxDistance=1200
 priority=0
 soundClassObject=&Fort_Creatures_Pets
 volumeMultiplier=0.8
 concurrencySet [/CosmeticCompanions/Assets/Audio_Shared/Foley/Quadraped/CONC_Mimosa_Quadraped_Footstep]
}
